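# Default Lit material for the Creative Random Number Generator device, making animated glowing dots that suggest the device is computing.
# A random UV noise function, fed by tiled V coordinates and a seed that changes each whole second, is compared in a hard SphereMask against a looping value quantized by Steps.
# That mask plus a bias scales the emissive color, which a Fresnel term then modulates.
# Base color, metallic, specular and roughness are plain parameters.
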Material /Game/Creative/Devices/RandomNumberGenerator/Materials/M_Device_RNG_Dots
graph {
 properties {pixelDepthOffsetMode=PDOM_Legacy, usedShadingModels=Default Lit}
 nodes [
  n1 = VectorParameter() {parameterName=Emissive, defaultValue=<rgba 0.217 1.2701 5 0>}
  n2 = ScalarParameter() {defaultValue=1, parameterName=EmissiveScale}
  n3 = TextureCoordinate() {vTiling=10}
  n4 = Time() {bOverride_Period=true, period=1000}
  n5 = Frac(None=Multiply(A=Time() {period=1}, B=ScalarParameter() {defaultValue=0.75, parameterName=ComputeSpeed}))
  n6 = ScalarParameter() {defaultValue=7, parameterName=Steps}
  n7 = Divide(A=Floor(None=Multiply(A=@n5, B=@n6)), B=@n6)
  n8 = Divide(B=@n6) {constA=0.5}
  n9 = SphereMask(
   A=MaterialFunctionCall!(
    PrimaryTextureCoordinates=ComponentMask(None=@n3) {g=true}
    2ndOrderTextureCoordinates
    RandomSeed=Ceil(None=@n4)
    NoiseDensityBias
    Bypass Noise Influence Crosstalk
   ) {materialFunction=/Game/Effects/Fort_Effects/Materials/_Functions/MF_RandomUVSample}.NoiseWithRange
   B=Multiply(A=Add(A=@n7, B=@n8), B=Divide(A=@n6) {constB=10}) {constB=0.7}
   Radius=Divide(A=@n6) {constB=100}
   Hardness=Constant() {r=1}
  )
  n10 = ScalarParameter() {parameterName=ComputeEmissiveBias}
  Multiply(
   A=Multiply(A=Multiply(A=@n1.RGB, B=@n2), B=Add(A=@n9, B=@n10))
   B=MaterialFunctionCall(Exponent=ScalarParameter() {defaultValue=0.5, parameterName=FresnelExponent}) {materialFunction=&MF_Fresnel_FX}.Fresnel
  )
  ScalarParameter() {parameterName=Metallic}
  ScalarParameter() {defaultValue=0.8, parameterName=Specular}
  ScalarParameter() {defaultValue=0.1, parameterName=Roughness}
  VectorParameter() {parameterName=BaseColor, defaultValue=<rgba 0.0011 0.0109 0.0434 0>}
 ]
}
