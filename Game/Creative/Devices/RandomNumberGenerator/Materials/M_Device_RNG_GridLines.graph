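# Unlit, two-sided translucent material for the Creative Random Number Generator device, drawing subdividing grid lines and a dot pattern.
# Sine and sphere masks on tiled U coordinates form soft lines, with object X scale compensating their width.
# A second sphere mask tiles dots using object X/Y scale.
# Lines, base opacity, inverted Fresnel and dots sum into depth-faded opacity, while the line mask blends base and line colors.
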Material /Game/Creative/Devices/RandomNumberGenerator/Materials/M_Device_RNG_GridLines
graph {
 properties {
  bEnableResponsiveAA=true
  blendMode=BLEND_Translucent
  pixelDepthOffsetMode=PDOM_Legacy
  shadingModel=MSM_Unlit
  twoSided=true
  usedShadingModels=Unlit
 }
 nodes [
  n1 = ScalarParameter() {defaultValue=1, parameterName=Subdivisions}
  n2 = Frac(None=Multiply(A=TextureCoordinate(), B=@n1))
  n3 = ScalarParameter() {defaultValue=0.02, parameterName=LineRadius}
  n4 = Divide(A=@n1)
  n5 = MaterialFunctionCall() {materialFunction=/Engine/Functions/Engine_MaterialFunctions02/WorldPositionOffset/ObjectScale}
  n6 = Divide(A=@n5.Scale X) {constB=512}
  n7 = SphereMask(
   A=Sine(None=MaterialFunctionCall(Float2=@n2) {materialFunction=&BreakOutFloat2Components}.R) {period=2}
   B=Constant()
   Radius=Divide(A=Multiply(A=@n3, B=@n4), B=@n6)
   Hardness=ScalarParameter() {defaultValue=0.6, parameterName=LineHardness}
  )
  n8 = Add(A=@n7, B=ScalarParameter() {defaultValue=0.05, parameterName=OpacityBase})
  n9 = MaterialFunctionCall(
   Brightness=ScalarParameter() {defaultValue=0.5, parameterName=FresnelBrightness}
   Exponent=ScalarParameter() {defaultValue=4, parameterName=FresnelExponent}
  ) {materialFunction=&MF_Fresnel_FX}
  n10 = AppendVector(A=@n6, B=Divide(A=@n5.Scale Y) {constB=512})
  n11 = TextureCoordinate()
  n12 = Frac(None=Multiply(A=Multiply(A=@n10, B=@n11), B=ScalarParameter() {defaultValue=40, parameterName=DotsRepeat}))
  n13 = Constant() {r=0.5}
  n14 = ScalarParameter() {defaultValue=0.1, parameterName=Radius}
  n15 = ScalarParameter() {defaultValue=0.5, parameterName=DotsHardness}
  DepthFade(
   Opacity=Add(A=Add(A=@n8, B=@n9.Fresnel Inverted Double Sided), B=SphereMask(A=@n12, B=@n13, Radius=@n14, Hardness=@n15))
   FadeDistance=ScalarParameter() {defaultValue=4, parameterName=DepthFadeDistance}
  )
  LinearInterpolate(
   A=VectorParameter() {parameterName=ColorBase, defaultValue=<rgba 0.2424 0.3001 0.9635 0>}.RGB
   B=VectorParameter() {parameterName=ColorLines, defaultValue=<rgba 0.6787 0.8045 0.9635 0>}.RGB
   Alpha=@n7
  )
 ]
}
