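# Material for the Random Number Generator device's glowing three-digit number display.
# Samples a digit-strip texture three times, shifted by the Thousandth, Hundreth and Tenth parameters, with sphere masks isolating each slot.
# Compute blends in a stepped, time-driven UV offset that rapidly cycles the digits.
# Digits plus a faint radial background are tinted by Emissive and modulated by a tiled dot pattern for an LED-panel look.
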Material /Game/Creative/Devices/RandomNumberGenerator/Materials/M_Device_RNG_Number
graph {
 properties {pixelDepthOffsetMode=PDOM_Legacy, usedShadingModels=Default Lit}
 nodes [
  n1 = TextureCoordinate() {uTiling=3}
  n2 = TextureCoordinate() {uTiling=3}
  n3 = AppendVector(A=Constant() {r=1}, B=ScalarParameter() {defaultValue=0.6, parameterName=V scale})
  n4 = Time() {bOverride_Period=true, period=1000}
  n5 = Divide(A=Ceil(None=Multiply(A=@n4) {constB=25})) {constB=10}
  n6 = Constant()
  n7 = Add(
   A=Multiply(
    A=MaterialFunctionCall(UVs=@n2, Texture Scale=@n3) {materialFunction=&ScaleUVsByCenter}.UVs
    B=Constant2Vector() {g=1, r=0.1}
   )
   B=Multiply(A=AppendVector(A=@n5, B=@n6), B=ScalarParameter() {parameterName=Compute})
  )
  n8 = AppendVector(A=ScalarParameter() {parameterName=Thousandth}, B=Constant())
  n9 = TextureCoordinate() {uTiling=3}
  n10 = Constant() {r=0.5}
  n11 = Multiply(
   A=TextureSample(UVs=Add(A=@n7, B=@n8)) {samplerType=SAMPLERTYPE_Masks, texture=/Game/Creative/Devices/Trigger/Materials/T_NumberSet_Large_CreativeTrigger}.R
   B=SphereMask(
    A=MaterialFunctionCall(Float2=@n9) {materialFunction=&BreakOutFloat2Components}.R
    B=@n10
    Radius=@n10
    Hardness=Constant() {r=1}
   )
  )
  n12 = Add(A=@n7, B=Constant2Vector() {r=-0.1}) {constB=-0.1}
  n13 = AppendVector(A=ScalarParameter() {parameterName=Hundreth}, B=Constant())
  n14 = TextureCoordinate() {uTiling=3}
  n15 = Multiply(
   A=TextureSample(UVs=Add(A=@n12, B=@n13)) {samplerType=SAMPLERTYPE_Masks, texture=&T_NumberSet_Large_CreativeTrigger}.R
   B=SphereMask(
    A=MaterialFunctionCall(Float2=@n14) {materialFunction=&BreakOutFloat2Components}.R
    B=Constant() {r=1.5}
    Radius=Constant() {r=0.5}
    Hardness=Constant() {r=1}
   )
  )
  n16 = Constant2Vector() {r=-0.2}
  n17 = ScalarParameter() {parameterName=Tenth}
  n18 = Constant()
  n19 = TextureSample(UVs=Add(A=Add(A=@n7, B=@n16) {constB=-0.1}, B=AppendVector(A=@n17, B=@n18))) {samplerType=SAMPLERTYPE_Masks, texture=&T_NumberSet_Large_CreativeTrigger}
  n20 = SphereMask(
   A=MaterialFunctionCall(Float2=TextureCoordinate() {uTiling=3}) {materialFunction=&BreakOutFloat2Components}.R
   B=Constant() {r=2.5}
   Radius=Constant() {r=0.5}
   Hardness=Constant() {r=1}
  )
  n21 = Multiply(
   A=SphereMask(
    A=MaterialFunctionCall(Float2=@n1) {materialFunction=&BreakOutFloat2Components}.G
    B=Constant() {r=0.5}
    Radius=Constant() {r=0.3}
    Hardness=Constant() {r=1}
   )
   B=Add(A=Add(A=@n11, B=@n15), B=Multiply(A=@n19.R, B=@n20))
  )
  n22 = Multiply(
   A=SphereMask(A=TextureCoordinate(), B=Constant() {r=0.5}, Radius=ScalarParameter() {defaultValue=0.7, parameterName=BGRadius}, Hardness=Constant())
   B=ScalarParameter() {defaultValue=0.04, parameterName=BGScale}
  )
  n23 = VectorParameter() {parameterName=Emissive, defaultValue=<rgba 0.1719 1.0803 3 0>}
  n24 = ScalarParameter() {defaultValue=1, parameterName=EmissiveScale}
  n25 = Frac(None=TextureCoordinate() {uTiling=45, vTiling=45})
  n26 = Constant() {r=0.5}
  n27 = Constant() {r=0.55}
  n28 = Constant()
  Multiply(
   A=Multiply(A=Add(A=@n21, B=@n22), B=Multiply(A=@n23.RGB, B=@n24))
   B=Add(A=SphereMask(A=@n25, B=@n26, Radius=@n27, Hardness=@n28)) {constB=0.1}
  )
  ScalarParameter() {parameterName=Metallic}
  ScalarParameter() {defaultValue=0.8, parameterName=Specular}
  ScalarParameter() {defaultValue=0.1, parameterName=Roughness}
  VectorParameter() {parameterName=BaseColor, defaultValue=<rgba 0.0011 0.0109 0.0434 0>}
 ]
}
